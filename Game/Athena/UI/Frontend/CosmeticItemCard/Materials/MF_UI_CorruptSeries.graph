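# UI material function for cosmetic item cards that draws the Corrupt series background when bIsCorruptSeries is on, otherwise outputting its input.
# Blended screen-aligned and local UI UVs form a grid whose time-seeded hash sizes and toggles squares, with a panning vertical band enlarging them.
# Glow-edged cells fade radially into GradientColor, then blend with an aspect-corrected SeriesTexture by Background, which defaults to texture only.
# Actor-anchored screen-space UV and glow-power chains are unconnected.
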MaterialFunction /Game/Athena/UI/Frontend/CosmeticItemCard/Materials/MF_UI_CorruptSeries
graph {nodes [
  n1 = MaterialFunctionCall(Float2=MaterialFunctionCall() {materialFunction=&ScreenResolution}.Visible Resolution) {materialFunction=&BreakOutFloat2Components}
  n2 = Max(A=@n1.G)
  n3 = Multiply(
   A=AppendVector(A=Divide(A=@n1.R, B=@n2), B=Constant() {r=1})
   B=MaterialFunctionCall() {materialFunction=&ScreenAlignedUVs}."X 100%, Y 100%"
  )
  n4 = ScalarParameter() {defaultValue=4, group=CorruptSeries, parameterName=Global UV Foreground}
  n5 = MaterialFunctionCall() {materialFunction=&GetUserInterfaceUV}
  n6 = MaterialFunctionCall(Float2=@n5.Pixel Size) {materialFunction=&BreakOutFloat2Components}
  n7 = Max(A=@n6.G)
  n8 = Multiply(A=AppendVector(A=Divide(A=@n6.R, B=@n7), B=Constant() {r=1}), B=@n5.Normalized UV)
  n9 = ScalarParameter() {defaultValue=0.25, group=CorruptSeries, parameterName=Localized UV}
  n10 = Multiply(
   A=Add(A=Multiply(A=@n3, B=@n4), B=Multiply(A=@n8, B=@n9))
   B=ScalarParameter() {defaultValue=12, group=CorruptSeries, parameterName=UV_Virus_SquarePatternScale}
  )
  n11 = Subtract(A=Multiply(A=Frac(None=@n10)) {constB=2})
  n12 = Abs(None=ComponentMask(None=@n11) {r=true})
  n13 = Constant()
  n14 = MaterialFunctionCall(
   Seed=Multiply(A=Time(), B=ScalarParameter() {defaultValue=0.15, group=CorruptSeries, parameterName=CorruptionPatternSpeed}) {constB=0.25}
   Vec2=Ceil(None=@n10)
  ) {materialFunction=&MF_Hash}
  n15 = Clamp(None=ComponentMask(None=ConstantBiasScale(None=@n14.Vec2)) {r=true})
  n16 = LinearInterpolate(
   A=ScalarParameter() {defaultValue=0.25, group=CorruptSeries, parameterName=SquareScaleMin}
   B=ScalarParameter() {defaultValue=0.9, group=CorruptSeries, parameterName=SquareScaleMax}
   Alpha=@n15
  )
  n17 = MaterialFunctionCall() {materialFunction=&GetUserInterfaceUV}
  n18 = ComponentMask(None=@n17.Normalized UV) {g=true}
  n19 = Frac(None=Multiply(A=Time(), B=ScalarParameter() {defaultValue=-0.15, group=CorruptSeries, parameterName=PatternPanSpeed}) {constB=0.25})
  n20 = SphereMask(A=Subtract(A=@n18, B=Multiply(A=@n19) {constB=2}), B=Constant(), Radius=Constant() {desc=1, r=0.5}, Hardness=Constant())
  n21 = Multiply(A=@n15, B=ScalarParameter() {defaultValue=0.5, group=CorruptSeries, parameterName=cELLsTRENGTH})
  n22 = Clamp(None=Round(None=Subtract(A=@n20, B=@n21)))
  n23 = LinearInterpolate(A=@n16, B=Multiply(A=@n16, B=ScalarParameter() {defaultValue=1.5, group=CorruptSeries, parameterName=SquarePanScale}), Alpha=@n22)
  n24 = Constant() {r=1}
  n25 = Abs(None=ComponentMask(None=@n11) {g=true})
  n26 = Multiply(A=SphereMask(A=@n12, B=@n13, Radius=@n23, Hardness=@n24), B=SphereMask(A=@n25, B=@n13, Radius=@n23, Hardness=@n24))
  n27 = ComponentMask(None=Clamp(None=@n14.Vec2)) {g=true}
  n28 = Add(A=Ceil(None=@n27), B=@n22)
  n29 = LinearInterpolate(
   A=FunctionInput(Preview=VectorParameter() {group=CorruptSeries, parameterName=BG_Color, defaultValue=<rgba 1 1 1 1>}.RGB) {bUsePreviewValueAsDefault=true, inputName=RGB 1, sortPriority=1}
   B=FunctionInput(Preview=VectorParameter() {group=CorruptSeries, parameterName=CellColor, defaultValue=<rgba 0 0 0 1>}.RGB) {bUsePreviewValueAsDefault=true, inputName=RGB 2, sortPriority=1}
   Alpha=Multiply(A=@n26, B=Clamp(None=@n28))
  )
  n30 = VectorParameter() {group=CorruptSeries, parameterName=GlowColor, defaultValue=<rgba 1 0.0079 0 1>}
  n31 = Add(A=@n23, B=ScalarParameter() {defaultValue=0.15, group=CorruptSeries, parameterName=GlowEdgeWidth})
  n32 = Multiply(A=SphereMask(A=@n12, B=@n13, Radius=@n31), B=SphereMask(A=@n25, B=@n13, Radius=@n31))
  n33 = Round(None=Power(Base=@n27))
  n34 = Multiply(A=Clamp(None=Subtract(A=@n32, B=@n26)), B=Clamp(None=Add(A=@n33, B=@n22)))
  n35 = ScalarParameter() {defaultValue=1, parameterName=GradientStrength}
  n36 = SphereMask(
   A=@n17.Normalized UV
   B=Constant2Vector() {g=0.5, r=0.5}
   Radius=ScalarParameter() {defaultValue=0.25, parameterName=GradientRadius}
   Hardness=ScalarParameter() {parameterName=GradientHardness}
  )
  n37 = MaterialFunctionCall(Float2=TextureCoordinate() {coordinateIndex=3}) {materialFunction=&BreakOutFloat2Components}
  n38 = Divide(A=@n37.R, B=Max(A=@n37.G))
  n39 = Multiply(A=AppendVector(A=@n38, B=Constant() {r=1}), B=TextureCoordinate() {coordinateIndex=4})
  n40 = AppendVector(A=Multiply(A=OneMinus(None=@n38)) {constB=0.5}, B=Constant())
  n41 = LinearInterpolate(
   A=LinearInterpolate(
    A=VectorParameter() {parameterName=GradientColor, defaultValue=<rgba 0 0 0 1>}.RGB
    B=LinearInterpolate(A=@n29, B=@n30.RGB, Alpha=@n34)
    Alpha=Multiply(A=@n35, B=@n36)
   )
   B=TextureSampleParameter2D(UVs=Add(A=@n39, B=@n40)) {group=Series, parameterName=SeriesTexture, texture=/Game/Global/Textures/Default/Masks_and_Gradients/T_Ui_LavaSeries_Bg}.RGB
   Alpha=ScalarParameter() {defaultValue=1, parameterName=Background}
  )
  FunctionOutput(None=StaticSwitchParameter(True=Multiply(A=@n41), False=FunctionInput()) {parameterName=bIsCorruptSeries})
  n42 = MaterialFunctionCall(Float2=ViewSize()) {materialFunction=&BreakOutFloat2Components}
  n43 = AppendVector(A=Constant() {r=1}, B=Divide(A=@n42.G, B=@n42.G))
  n44 = Multiply(A=ConstantBiasScale(None=ScreenPosition().ViewportUV) {bias=-0.5, scale=2}, B=@n43)
  n45 = ComponentMask(None=PreSkinnedLocalBounds().Half Extents) {b=true}
  n46 = ScalarParameter() {group=Virus, parameterName=Virus_ScreenSpaceOffset}
  n47 = Add(
   A=ActorPositionWS()
   B=MaterialFunctionCall(Z=Add(A=@n45, B=@n46)) {materialFunction=&MakeFloat3}.Result
  ) {constB=100}
  n48 = Constant()
  n49 = Constant()
  n50 = Multiply(
   A=Multiply(
    A=Constant2Vector() {g=-0.5, r=0.5}
    B=MaterialFunctionCall(World Position=@n47) {materialFunction=&TransformToClipSpace}.Clip Space XY
   )
   B=If(A=Distance(A=@n47, B=@n48), B=Constant() {r=0.01}, "A > B"=Constant() {r=1}, "A == B"=@n49, "A < B"=@n49)
  )
  n51 = Constant() {r=100}
  n52 = ComponentMask(None=TransformPosition(None=@n47) {transformSourceType=TRANSFORMPOSSOURCE_World, transformType=TRANSFORMPOSSOURCE_View}) {b=true}
  MaterialFunctionCall(
   UVs=Subtract(A=ConstantBiasScale(None=@n44), B=Multiply(A=@n43, B=@n50))
   Texture Scale=Multiply(A=Divide(A=@n51, B=@n52), B=ScalarParameter() {defaultValue=128, parameterName=Virus_SquarePatternScale})
  ) {materialFunction=&ScaleUVsByCenter}
  ObjectRadius()
  Power(Base=Multiply(A=Multiply(A=@n30.RGB, B=@n34), B=ScalarParameter() {defaultValue=5, group=CorruptSeries, parameterName=GlowSterngth}))
  Reroute(None=@n18)
  Round(None=@n27)
  ScalarParameter() {defaultValue=4, group=CorruptSeries, parameterName=Global UV Background}
  TextureCoordinate()
  TextureCoordinate()
 ]}
