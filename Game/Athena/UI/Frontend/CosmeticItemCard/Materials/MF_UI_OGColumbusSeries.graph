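# UI material function for the OG Columbus series look on cosmetic item cards, switched by bIsOGColumbusSeries with an input passthrough otherwise.
# Radially panned splatter layers, masked by distance, lerp via bIsStatic against two aspect-corrected screen-aligned starfield layers with optional local-UV influence.
# Background then blends in SeriesTexture, sampled on UV4 with centered aspect correction from UV3.
# The RGB 1 and RGB 2 inputs are unconnected.
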MaterialFunction /Game/Athena/UI/Frontend/CosmeticItemCard/Materials/MF_UI_OGColumbusSeries
graph {nodes [
  ComponentMask() {g=true, r=true}
  FunctionInput(Preview=Constant3Vector() {constant=<rgba 0 0.9473 0.9144 1>}) {inputName=RGB 1, sortPriority=1}
  FunctionInput(Preview=Constant3Vector() {constant=<rgba 0 0.1207 0.5573 1>}) {inputName=RGB 2, sortPriority=1}
  n1 = MaterialFunctionCall() {materialFunction=&VectorToRadialValue}
  n2 = Panner(
   Coordinate=@n1.Radial Coordinates
   Speed=AppendVector(A=ScalarParameter() {defaultValue=0.0125, parameterName=U-Speed}, B=ScalarParameter() {defaultValue=-2, parameterName=V-Speed})
  )
  n3 = AppendVector(A=ScalarParameter() {defaultValue=1, parameterName=R-U-Tiles}, B=ScalarParameter() {defaultValue=0.01, parameterName=R-V-Tiles})
  n4 = Multiply(A=TextureSample(UVs=Multiply(A=@n2, B=@n3)) {texture=/Game/Global/Textures/Default/PointSplatter/T_Ui_ColumbusSeries_512}.RGB) {constB=2}
  n5 = AppendVector(A=ScalarParameter() {defaultValue=3, parameterName=G-U-Tiles}, B=ScalarParameter() {defaultValue=0.03, parameterName=G-V-Tiles})
  n6 = Multiply(A=TextureSample(UVs=Multiply(A=@n2, B=@n5)) {texture=&T_Ui_ColumbusSeries_512}.RGB) {constB=2}
  n7 = ScalarParameter() {defaultValue=2, parameterName=MaskPower}
  n8 = TextureCoordinate()
  n9 = ScalarParameter() {parameterName=R-Local-UV-Influence}
  n10 = MaterialFunctionCall(Float2=MaterialFunctionCall() {materialFunction=&ScreenResolution}.Visible Resolution) {materialFunction=&BreakOutFloat2Components}
  n11 = Divide(A=@n10.R, B=Max(A=@n10.G) {constB=0.1})
  n12 = Constant() {r=1}
  n13 = Multiply(
   A=Multiply(
    A=AppendVector(A=@n11, B=@n12)
    B=MaterialFunctionCall() {materialFunction=&ScreenAlignedUVs}."X 100%, Y 100%"
   )
   B=ScalarParameter() {defaultValue=1, parameterName=Global Tile}
  )
  n14 = TextureSample(UVs=Add(A=Multiply(A=@n8, B=@n9), B=@n13)) {texture=/Game/Global/Textures/Default/PointSplatter/T_UI_Starfield}
  n15 = ScalarParameter() {parameterName=R-Strength}
  n16 = TextureCoordinate()
  n17 = ScalarParameter() {parameterName=G-Local-UV-Influence}
  n18 = TextureSample(UVs=Add(A=@n13, B=Multiply(A=@n16, B=@n17))) {texture=&T_UI_Starfield}
  n19 = ScalarParameter() {parameterName=G-Strength}
  n20 = LinearInterpolate(
   A=Multiply(A=Power(Base=@n1.Linear Distance, Exp=@n7), B=Add(A=@n4, B=@n6))
   B=Add(A=Multiply(A=@n14.RGB, B=@n15) {constB=2}, B=Multiply(A=@n18.RGB, B=@n19) {constB=2})
   Alpha=ScalarParameter() {defaultValue=1, parameterName=bIsStatic}
  )
  n21 = MaterialFunctionCall(Float2=TextureCoordinate() {coordinateIndex=3}) {materialFunction=&BreakOutFloat2Components}
  n22 = Divide(A=@n21.R, B=Max(A=@n21.G))
  n23 = AppendVector(A=@n22, B=Constant() {r=1})
  n24 = TextureCoordinate() {coordinateIndex=4}
  n25 = Multiply(A=OneMinus(None=@n22)) {constB=0.5}
  n26 = Constant()
  n27 = TextureSampleParameter2D(UVs=Add(A=Multiply(A=@n23, B=@n24), B=AppendVector(A=@n25, B=@n26))) {group=Series, parameterName=SeriesTexture, texture=/Game/Global/Textures/Default/Masks_and_Gradients/T_Ui_LavaSeries_Bg}
  n28 = ScalarParameter() {parameterName=Background}
  FunctionOutput(None=StaticSwitchParameter(True=LinearInterpolate(A=@n20, B=@n27.RGB, Alpha=@n28), False=FunctionInput()) {parameterName=bIsOGColumbusSeries})
 ]}
