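# Material function for UI that optionally applies the Rocket Pass series effect to an item preview material.
# A static switch, bIsRocketPassSeries, either passes the input through unchanged or outputs the MF_UI_RocketPass result.
# The Rocket Pass branch feeds MF_UI_RocketPass with scaled-to-fit UVs from MF_UI_UVScaling.
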MaterialFunction /Game/Environments/FrontEnd/Blueprints/ItemPreview/RocketPass/Materials/MF_UI_RocketPassSeries
graph {nodes [
  n1 = MaterialFunctionCall() {materialFunction=&MF_UI_UVScaling}
  FunctionOutput(None=StaticSwitchParameter(
    True=MaterialFunctionCall!(UVs=@n1.Scaled to Fit, Floating Lerp) {materialFunction=&/MF_UI_RocketPass}.Result
    False=FunctionInput()
   ) {parameterName=bIsRocketPassSeries})
 ]}
