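# Master translucent UI material for the frontend cosmetic item card, drawing an item preview over a styled background with rarity trim.
# Chained series functions build the background, then the aspect-corrected preview, plain or channel-packed, is composited with a vignette.
# An angled rarity band tinted by the image, an enhanced plus icon, and variant and archived gradients layer on top.
# Opacity multiplies an edge or angled reward mask by a rounded-corner mask.
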Material /FortniteCoreUI/Athena/UI/Frontend/CosmeticItemCard/Materials/M_UI_ItemCard_V2_Mst
graph {
 properties {
  bUsedWithStaticMesh=false
  blendMode=BLEND_Translucent
  materialDomain=MD_UI
  pixelDepthOffsetMode=PDOM_Legacy
  usedShadingModels=Default Lit
 }
 nodes [
  n1 = Constant() {r=2}
  n2 = Constant() {r=0.75}
  n3 = VectorParameter() {group=Background, parameterName=SingleRGB, defaultValue=<rgba 0.7135 0 0.0457 1>}
  n4 = LinearInterpolate(
   A=@n3.RGB
   B=Power(Base=Multiply(A=@n3.RGB) {constB=0.5})
   Alpha=MaterialFunctionCall(GradientEnd=Constant() {r=1}, RotationAngle=Constant() {r=0.75}) {materialFunction=&MF_UI_LinearGradient}.Result
  )
  n5 = ScalarParameter() {group=Background, parameterName=VariantColor}
  n6 = ScalarParameter() {parameterName=bIsShuffIeItem}
  n7 = MaterialFunctionCall(Float2=TextureCoordinate() {coordinateIndex=3}) {materialFunction=&BreakOutFloat2Components}
  n8 = Divide(A=@n7.R, B=Max(A=@n7.G))
  n9 = Constant() {r=1}
  n10 = OneMinus(None=@n8)
  n11 = Add(
   A=Multiply(A=AppendVector(A=@n8, B=@n9), B=TextureCoordinate() {coordinateIndex=4})
   B=AppendVector(A=Multiply(A=@n10) {constB=0.5}, B=Constant())
  )
  n12 = MaterialFunctionCall(
   UVs=@n11
   Texture Scale=Multiply(A=ScalarParameter() {defaultValue=1, group=Item, parameterName=ScalePreviewImage}, B=LinearInterpolate(Alpha=@n6) {constA=1, constB=0.85})
  ) {materialFunction=&ScaleUVsByCenter}
  n13 = TextureSampleParameter2D(UVs=@n12.UVs) {group=Item, parameterName=ItemDisplay, texture=/Game/Global/Textures/Default/DefaultUI/T-Soldier-HID-031-Athena-Commando-M-Retro-L}
  n14 = Add(A=Add(A=@n13.R, B=@n13.R), B=@n13.B)
  n15 = MaterialFunctionCall!(UVs=@n12.UVs, Alpha=LinearInterpolate(A=@n13.A, B=Saturate(None=@n14), Alpha=ScalarParameter() {parameterName=IsChannelPacked})) {materialFunction=/Game/UI/Foundation/Items/ClampAlphaToTransparentBorder}
  n16 = VectorParameter() {group=Background, parameterName=RGB 1, defaultValue=<rgba 0 0.4667 1 1>}
  n17 = VectorParameter() {group=Background, parameterName=RGB 2, defaultValue=<rgba 0.0027 0.058 0.462 1>}
  n18 = MaterialFunctionCall!(
   In=MaterialFunctionCall!(
    RGB 1=@n16.RGB
    RGB 2=@n17.RGB
    In=MaterialFunctionCall!(RGB 2=@n17.RGB, RGB 1=@n16.RGB) {materialFunction=/Game/Athena/UI/Frontend/CosmeticItemCard/Materials/MF_UI_Default}.Result
   ) {materialFunction=/Game/Athena/UI/Frontend/CosmeticItemCard/Materials/MF_UI_LavaSeries}.Result
   RGB 2=@n17.RGB
   RGB 1=@n16.RGB
  ) {materialFunction=/Game/Athena/UI/Frontend/CosmeticItemCard/Materials/MF_UI_FrozenSeries}
  n19 = MaterialFunctionCall!(
   In=MaterialFunctionCall!(
    In=MaterialFunctionCall!(In=@n18.Result, RGB 2=@n17.RGB, RGB 1=@n16.RGB) {materialFunction=/Game/Athena/UI/Frontend/CosmeticItemCard/Materials/MF_UI_MarvelSeries}.Result
    RGB 2=@n17.RGB
    RGB 1=@n16.RGB
   ) {materialFunction=/Game/Athena/UI/Frontend/CosmeticItemCard/Materials/MF_UI_CubeSeries}.Result
   RGB 2=@n17.RGB
   RGB 1=@n16.RGB
  ) {materialFunction=/Game/Athena/UI/Frontend/CosmeticItemCard/Materials/MF_UI_ShadowSeries}
  n20 = MaterialFunctionCall!(
   In=MaterialFunctionCall!(
    In=MaterialFunctionCall!(In=@n19.Result, RGB 2=@n17.RGB, RGB 1=@n16.RGB) {materialFunction=/Game/Athena/UI/Frontend/CosmeticItemCard/Materials/MF_UI_CreatorCollabSeries}.Result
    RGB 2=@n17.RGB
    RGB 1=@n16.RGB
   ) {materialFunction=/Game/Athena/UI/Frontend/CosmeticItemCard/Materials/MF_UI_DCSeries}.Result
   RGB 2=@n17.RGB
   RGB 1=@n16.RGB
  ) {materialFunction=/Game/Athena/UI/Frontend/CosmeticItemCard/Materials/MF_UI_SlurpSeries}
  n21 = MaterialFunctionCall!(
   In=MaterialFunctionCall!(
    In=MaterialFunctionCall!(In=@n20.Result, RGB 2=@n17.RGB, RGB 1=@n16.RGB) {materialFunction=/Game/Athena/UI/Frontend/CosmeticItemCard/Materials/MF_UI_ColumbusSeries}.Result
    RGB 2=@n17.RGB
    RGB 1=@n16.RGB
   ) {materialFunction=/Game/Athena/UI/Frontend/CosmeticItemCard/Materials/MF_UI_CorruptSeries}.Result
   RGB 2=@n17.RGB
   RGB 1=@n16.RGB
  ) {materialFunction=/Game/Athena/UI/Frontend/CosmeticItemCard/Materials/MF_UI_PlatformSeries}
  n22 = Add(
   A=MaterialFunctionCall!(In=MaterialFunctionCall!(In=@n21.Result, RGB 2=@n17.RGB, RGB 1=@n16.RGB) {materialFunction=/Game/Athena/UI/Frontend/CosmeticItemCard/Materials/MF_UI_OGColumbusSeries}.Result) {materialFunction=/Game/Environments/FrontEnd/Blueprints/ItemPreview/RocketPass/Materials/MF_UI_RocketPassSeries}.Result
   B=ScalarParameter() {defaultValue=0.0125, group=Background, parameterName=Background-Balance, sortPriority=1}
  ) {constB=0.25}
  n23 = LinearInterpolate(B=VectorParameter() {group=ChannelPacked, parameterName=DynamicColor, defaultValue=<rgba 1 0 0.5663 0>}.RGB, Alpha=@n13.B)
  n24 = LinearInterpolate(
   A=@n13.RGB
   B=LinearInterpolate(A=LinearInterpolate(A=@n23, Alpha=@n13.G), Alpha=@n13.R) {constB=0}
   Alpha=ScalarParameter() {group=ChannelPacked, parameterName=IsChannelPacked}
  )
  n25 = MaterialFunctionCall(
   CenterPosition=Constant2Vector() {r=0.5}
   Radius=ScalarParameter() {defaultValue=1.35, group=Background, parameterName=Vignette Radius}
   Density=ScalarParameter() {defaultValue=3, group=Background, parameterName=Vignette Density}
  ) {materialFunction=&RadialGradientExponential}
  n26 = LinearInterpolate(
   A=LinearInterpolate(A=@n22, B=@n24, Alpha=@n15.Result)
   B=@n17.RGB
   Alpha=Multiply(A=OneMinus(None=@n25.RadialGradientExponential), B=ScalarParameter() {defaultValue=1.25, group=Background, parameterName=Vignette-Strength})
  )
  n27 = Saturate(None=ComponentMask(None=TextureCoordinate() {coordinateIndex=4, uTiling=0.5}) {r=true})
  n28 = VectorParameter() {group=Rarity, parameterName=Rarity 1, defaultValue=<rgba 0 1 0.9647 1>}
  n29 = StaticSwitchParameter(
   True=LinearInterpolate(
    A=VectorParameter() {parameterName=Rarity 1 Overide, defaultValue=<rgba 1 0 0 1>}.RGB
    B=VectorParameter() {parameterName=Rarity 2 Overide, defaultValue=<rgba 0 1 0 1>}.RGB
    Alpha=@n27
   )
   False=LinearInterpolate(A=@n28.RGB, B=VectorParameter() {group=Rarity, parameterName=Rarity 2, defaultValue=<rgba 0 0.0976 0.521 1>}.RGB, Alpha=@n27)
  ) {parameterName=bCustomRarity-Bar}
  n30 = Multiply(A=@n24, B=@n15.Result)
  n31 = ScalarParameter() {defaultValue=0.5, group=Rarity, parameterName=Portrait-RangeFactor, sortPriority=1}
  n32 = Multiply(A=Clamp(None=Multiply(A=@n30, B=@n31)) {maxDefault=0.75, minDefault=0.125})
  n33 = ComponentMask(None=Rotator(
    Coordinate=TextureCoordinate() {coordinateIndex=4}
    Time=ScalarParameter() {defaultValue=0.01, group=Rarity, parameterName=Angle, sortPriority=1}
   ) {speed=6.2832}) {g=true}
  n34 = ScalarParameter() {defaultValue=0.9, group=Rarity, parameterName=Angle Offset, sortPriority=1}
  n35 = Subtract(A=@n33, B=@n34)
  n36 = ScalarParameter() {defaultValue=128, group=Rarity, parameterName=Angle Edge, sortPriority=1}
  n37 = Multiply(A=Saturate(None=Multiply(A=@n35, B=@n36)), B=LinearInterpolate(Alpha=ScalarParameter() {parameterName=bIsVariant}) {constA=1, constB=0})
  n38 = LinearInterpolate(
   A=@n26
   B=MaterialFunctionCall(Base=@n29, Blend=Desaturation(None=@n32) {luminanceFactors=<rgba 1 1 1 0>}) {materialFunction=&Blend_Overlay}.Result
   Alpha=@n37
  )
  n39 = ScalarParameter() {defaultValue=0.035, group=Rarity, parameterName=Angle Offset Add}
  n40 = Multiply(A=Subtract(A=@n33, B=Add(A=@n34, B=@n39)), B=@n36)
  n41 = DotProduct(A=@n37, B=OneMinus(None=Saturate(None=@n40)))
  n42 = StaticSwitchParameter(
   True=@n26
   False=LinearInterpolate(A=LinearInterpolate(A=@n38, B=@n29, Alpha=@n41), B=@n26, Alpha=ScalarParameter() {parameterName=bIsShuffIeItem})
  ) {parameterName=bIsVarient}
  n43 = ScalarParameter() {defaultValue=7, group=Enhanced, parameterName=Enhanced Scale}
  n44 = ScalarParameter() {defaultValue=-0.2, group=Enhanced, parameterName=Skew Factor}
  n45 = MaterialFunctionCall(Float2=@n11) {materialFunction=&BreakOutFloat2Components}
  n46 = OneMinus(None=@n45.G)
  n47 = AppendVector(A=Add(A=@n45.R, B=Multiply(A=@n44, B=@n46) {constB=-1}), B=@n45.G)
  n48 = ScalarParameter() {defaultValue=-4.85, group=Enhanced, parameterName=Enhance-X-Offset}
  n49 = ScalarParameter() {defaultValue=-5.76, group=Enhanced, parameterName=Enhance-Y-Offset}
  n50 = TextureSample(UVs=Add(A=Multiply(A=@n47, B=@n43), B=AppendVector(A=@n48, B=@n49))) {samplerSource=SSM_Clamp_WorldGroupSettings, texture=/Game/Athena/UI/Frontend/Parties/Art/T_UI_PlusSign_64x}
  n51 = LinearInterpolate(Alpha=ScalarParameter() {parameterName=bIsVariant}) {constA=1, constB=0}
  n52 = StaticSwitchParameter(True=VectorParameter() {parameterName=Rarity 1 Overide, defaultValue=<rgba 1 0 0 1>}.RGB, False=@n28.RGB) {parameterName=bCustomRarity-Bar}
  n53 = ScalarParameter() {defaultValue=0.5, group=Enhanced, parameterName=IconBrightness}
  n54 = LinearInterpolate(
   A=@n42
   B=LinearInterpolate(A=@n42, B=Add(A=@n52, B=@n53), Alpha=Multiply(A=@n50.A, B=@n51))
   Alpha=ScalarParameter() {group=Enhanced, parameterName=bIsEnhanced}
  )
  LinearInterpolate(
   A=StaticSwitchParameter(True=LinearInterpolate(A=@n54, B=@n4, Alpha=@n5), False=@n22) {defaultValue=true, parameterName=Display Character}
   B=LinearInterpolate(
    A=Constant3Vector()
    B=Constant3Vector() {constant=<rgba 0 0.0633 0.2083 0>}
    Alpha=MaterialFunctionCall(GradientEnd=@n1, RotationAngle=@n2) {materialFunction=&MF_UI_LinearGradient}.Result
   )
   Alpha=ScalarParameter() {parameterName=ArchivedGradient}
  )
  n55 = Rotator(Coordinate=TextureCoordinate(), Time=ScalarParameter() {defaultValue=-0.03, parameterName=AlphaAngle}) {speed=6.2832}
  n56 = ComponentMask(None=Abs(None=Add(A=@n55) {constB=-0.5})) {r=true}
  n57 = ScalarParameter() {defaultValue=0.385, parameterName=Subtract}
  n58 = OneMinus(None=Multiply(A=Subtract(A=@n56, B=@n57) {constB=0.25}, B=ScalarParameter() {defaultValue=256, parameterName=Multiply}) {constB=64})
  n59 = TextureCoordinate()
  n60 = MaterialFunctionCall(Float2=Abs(None=Add(A=@n59) {constB=-0.5})) {materialFunction=&BreakOutFloat2Components}
  n61 = ScalarParameter() {defaultValue=0.495, group=Background, parameterName=EdgeOuter}
  n62 = ScalarParameter() {defaultValue=128, group=Background, parameterName=EdgeOuterFalloff}
  n63 = Multiply(A=Subtract(A=@n60.R, B=@n61), B=@n62)
  n64 = Multiply(A=@n62, B=Subtract(A=@n60.G, B=@n61))
  n65 = Saturate(None=Add(A=Saturate(None=@n63), B=Saturate(None=@n64)))
  n66 = ScalarParameter() {
   defaultValue=1
   desc="If enabled reads corner radius values in pixels.\r\nbool"
   group=Layout
   parameterName=UseValueInPixels
   sliderMax=1
   sortPriority=1
  }
  n67 = MaterialFunctionCall(
   Corner Radius=MaterialFunctionCall(
    R=ScalarParameter() {group=Layout, parameterName=TopLeftCornerRadius, sortPriority=4}
    G=ScalarParameter() {group=Layout, parameterName=BottomLeftCornerRadius, sortPriority=7}
    B=ScalarParameter() {group=Layout, parameterName=TopRightCornerRadius, sortPriority=5}
    A=ScalarParameter() {group=Layout, parameterName=BottomRightCornerRadius, sortPriority=6}
   ) {materialFunction=&AppendMany}.RGBA
   Radius in Pixels=@n66
  ) {materialFunction=&MF_UI_CoreUI_RoundedMask}
  n68 = ScalarParameter() {defaultValue=1, group=Color, parameterName=Additive Percentage, sliderMax=1, sortPriority=3}
  Multiply(
   A=LinearInterpolate(A=OneMinus(None=@n65), B=Saturate(None=@n58), Alpha=ScalarParameter() {parameterName=bIsReward})
   B=NamedRerouteDeclaration(None=Multiply(A=@n67.Shape Mask, B=@n68)) {name=ShapeMask, nodeColor=<rgba 0 0 0 1>}
  )
 ]
}
