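# Frontend cosmetic item card widget, subclassing native FortCosmeticItemCard, drawn as one material-driven image inside a load guard.
# Parent events write icon, rarity colors and visibility, enhanced, reward, shuffle and random values into the dynamic material.
# Material swaps record initial angle values and reapply the cached icon.
# Animations handle archived state and hover, picking a variant by rarity visibility, with a repeater event track for scaling the icon through the material.
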WidgetBlueprint /FortniteCoreUI/Athena/UI/Frontend/CosmeticItemCard/FortCosmeticItemCard
animations [
 Anim-Hover {
  binding_count=2
  length_seconds=0.15
  object=&self:Anim-Hover
  sequence {
   bindings [
    Image_NewItemCard {
     id=#1
     possessed_class=FortLazyImage
     tracks [{
       type=MovieSceneWidgetMaterialTrack
       sections [{
         type=MovieSceneParameterSection
         channels [
          Enhance-Y-Offset {type=MovieSceneScriptingFloatChannel, keys [[0,-5.76],[3,-5.5]]}
          Vignette-Strength {type=MovieSceneScriptingFloatChannel, keys [[0,1.25],[3,0.75]]}
          IconBrightness {type=MovieSceneScriptingFloatChannel, keys [[0,0.5],[3,1]]}
          Angle Offset {type=MovieSceneScriptingFloatChannel, keys [[0,0.9],[3,0.925]]}
          Background-Balance {type=MovieSceneScriptingFloatChannel, keys [[0,0.0125],[3,0.025]]}
         ]
        }]
      }]
    }
    FortCosmeticItemCard_C_21 {
     id=#2
     possessed_class=FortCosmeticItemCard_C
     tracks [{
       type=MovieSceneFloatTrack
       display_name=Animation Icon Scale
       sections [{type=MovieSceneFloatSection, channels [None {type=MovieSceneScriptingFloatChannel, keys [[0,1],[3,1.1]]}]}]
      }]
    }
   ]
   tracks [{type=MovieSceneEventTrack, sections [{type=MovieSceneEventRepeaterSection, end_frame=3, start_frame=0}]}]
  }
 }
 Archived {
  binding_count=1
  length_seconds=0.3
  object=&self:Archived
  sequence {bindings [Image_NewItemCard {
     id=#3
     possessed_class=Image
     tracks [{
       type=MovieSceneWidgetMaterialTrack
       sections [{
         type=MovieSceneParameterSection
         end_frame=6
         start_frame=0
         channels [ArchivedGradient {type=MovieSceneScriptingFloatChannel, keys [[0,0],[6,0.9]]}]
        }]
      }]
    }]}
 }
 Anim-Hover-NoRarity {
  binding_count=2
  length_seconds=0.15
  object=&self:Anim-Hover-NoRarity
  sequence {
   bindings [
    Image_NewItemCard {
     id=#1
     possessed_class=FortLazyImage
     tracks [{
       type=MovieSceneWidgetMaterialTrack
       sections [{
         type=MovieSceneParameterSection
         channels [
          Enhance-Y-Offset {type=MovieSceneScriptingFloatChannel, keys [[0,-5.76],[3,-5.5]]}
          Vignette-Strength {type=MovieSceneScriptingFloatChannel, keys [[0,1.25],[3,0.75]]}
          IconBrightness {type=MovieSceneScriptingFloatChannel, keys [[0,0.5],[3,1]]}
          Background-Balance {type=MovieSceneScriptingFloatChannel, keys [[0,0.0125],[3,0.025]]}
         ]
        }]
      }]
    }
    FortCosmeticItemCard_C_21 {
     id=#2
     possessed_class=FortCosmeticItemCard_C
     tracks [{
       type=MovieSceneFloatTrack
       display_name=Animation Icon Scale
       sections [{type=MovieSceneFloatSection, channels [None {type=MovieSceneScriptingFloatChannel, keys [[0,1],[3,1.1]]}]}]
      }]
    }
   ]
   tracks [{type=MovieSceneEventTrack, sections [{type=MovieSceneEventRepeaterSection, end_frame=3, start_frame=0}]}]
  }
 }
]
graphs [
 ArchiveAnimation {nodes [
   n1 = ArchiveAnimation!() -> (then, IsArchived:bool=~, UseAnimations:bool=~) {is=title K2Node_FunctionEntry}
   n2 = Get Archived() -> (Archived:Widget Animation Object Reference)
   > PlayAnimation(
    InAnimation=@n2
    StartAtTime=SelectFloat(A=0, B=GetEndTime(self=@n2), bPickA=@n1.UseAnimations)
    PlayMode=Select(Option 0:EUMGSequencePlayMode Enum=Reverse, Option 1:EUMGSequencePlayMode Enum, Index=@n1) -> (ReturnValue:EUMGSequencePlayMode Enum)
   )
  ]}
 EventGraph {nodes [
   n1 = OnSmallPreviewImageUpdated!() -> (OutputDelegate, then, Icon:Texture 2D Soft Object Reference) {is=title K2Node_CustomEvent}
   > Branch(Condition=NOT Boolean(A=ShouldDisplayLargePreviewImage!(self:Fort Cosmetic Item Card Object Reference) -> (ReturnValue:bool)))
   > UpdateItemCardIconWithTexture2D!(execute, self:Fort Cosmetic Item Card Object Reference, Icon:Texture 2D Soft Object Reference=@n1.Icon) -> (then)
   n2 = OnLargePreviewImageUpdated!() -> (OutputDelegate, then, LargePreviewImage:Texture 2D Soft Object Reference) {is=title K2Node_CustomEvent}
   > Branch(Condition=ShouldDisplayLargePreviewImage())
   > UpdateItemCardIconWithTexture2D(Icon=@n2.LargePreviewImage)
   Event OnHover() {is=title K2Node_Event}
   n3 = Select(
    Option 0:Widget Animation Object Reference=Get Anim-Hover-NoRarity() -> (Anim-Hover-NoRarity:Widget Animation Object Reference)
    Option 1:Widget Animation Object Reference=Get Anim-Hover() -> (Anim-Hover:Widget Animation Object Reference)
    Index=Get bRarityVisible() -> (bRarityVisible:bool)
   ) -> (ReturnValue:Widget Animation Object Reference)
   > PlayAnimationForward(InAnimation=@n3)
   n4 = Event OnUpdateVisuals!() -> (
    OutputDelegate
    then
    bIsItem:bool
    bIsBorderAnimated:bool
    bIsHero:bool
    bReward:bool
    bIsShuffleItem:bool
    bShowAsArchived:bool
    bUseAnimations:bool
   ) {is=title K2Node_Event}
   > HandleUpdateVisuals!(
    execute
    self
    bIsItem:bool=@n4.bIsItem
    bIsBorderAnimated:bool=@n4.bIsBorderAnimated
    IsHero:bool=@n4.bIsHero
    IsReward:bool=@n4.bReward
    IsShuffleItem:bool=@n4.bIsShuffleItem
    ShowAsArchived:bool=@n4.bShowAsArchived
    UseAnimations:bool=@n4.bUseAnimations
   ) -> (then)
   n5 = Event OnUpdateEnhanced!() -> (OutputDelegate, then, bEnhanced:bool) {is=title K2Node_Event}
   > SetScalarParameterValue(
    self=Get ImageMaterial_NewItemCard() -> (ImageMaterial_NewItemCard:Material Instance Dynamic Object Reference)
    ParameterName=bIsEnhanced
    Value="To Float (Boolean)"(InBool=@n5.bEnhanced)
   )
   n6 = Event OnUpdateSeriesOrRarityVisibility!() -> (OutputDelegate, then, bIsVisible:bool) {is=title K2Node_Event}
   > HandleUpdateRarityVisibility!(execute, self, bIsVisible:bool=@n6.bIsVisible) -> (then)
   n7 = Event OnUpdateItemCardMaterial!() -> (OutputDelegate, then, Material:Material Interface Object Reference) {is=title K2Node_Event}
   > SetBrushFromMaterial(self=Get Image_NewItemCard() -> (Image_NewItemCard:Image Object Reference), Material=@n7.Material)
   > n8 = GetDynamicMaterial(self=Get Image_NewItemCard())
   > n9 = Set ImageMaterial_NewItemCard(ImageMaterial_NewItemCard:Material Instance Dynamic Object Reference=@n8) -> (Output_Get:Material Instance Dynamic Object Reference)
   > n10 = Get Scalar Parameter Value!(execute, self:Material Instance Dynamic Object Reference=@n9, ParameterName:Name=Angle) -> (then, ReturnValue:float) {is=title K2Node_CallFunction}
   > Set InitialAngle(InitialAngle:double=@n10) -> (Output_Get:double)
   > n11 = Get Scalar Parameter Value(self=@n9, ParameterName=Angle Offset)
   > Set InitialAngleOffset(InitialAngleOffset:double=@n11) -> (Output_Get:double)
   > n12 = Get!(execute:Exec) -> (then:Exec, else:Exec, CachedIcon:Texture 2D Object Reference)
   > Update Icon!(execute, self, Icon:Texture 2D Object Reference=@n12, Scale:double=Get CachedScale() -> (CachedScale:float)) -> (then)
   n13 = Event OnUpdateItemCardIcon!() -> (OutputDelegate, then, Icon:Texture 2D Object Reference, IconScale:float) {is=title K2Node_Event}
   > n14 = Set CachedIcon(CachedIcon:Texture 2D Object Reference=@n13.Icon) -> (Output_Get:Texture 2D Object Reference)
   > n15 = Set CachedScale(CachedScale:float=@n13.IconScale) -> (Output_Get:float)
   > Update Icon(Icon=@n14, Scale=@n15)
   n16 = Event OnUpdateSeriesOrRarityColors!() -> (OutputDelegate, then, ColorPalette:"Fort Color Palette Structure (by ref)") {is=title K2Node_Event}
   > HandleUpdateRarity!(execute, self, RarityColors:Fort Color Palette Structure=@n16.ColorPalette) -> (then)
   Event OnUnhover() {is=title K2Node_Event}
   > PlayAnimationReverse(InAnimation=@n3)
   Event OnSizeChanged!() -> (OutputDelegate, then, InCardWidth:float) {is=title K2Node_Event}
  ]}
 HandleUpdateRarity {nodes [
   HandleUpdateRarity!() -> (then, RarityColors:Fort Color Palette Structure) {is=title K2Node_FunctionEntry}
   n1 = Break Fort Color Palette!(FortColorPalette:"Fort Color Palette Structure (by ref)"=Get RarityColors() -> (RarityColors:Fort Color Palette Structure)) -> (
    Color1:Linear Color Structure
    Color2:Linear Color Structure
    Color3:Linear Color Structure
    Color4:Linear Color Structure
    Color5:Linear Color Structure
   ) {is=title K2Node_BreakStruct}
   n2 = Get ImageMaterial_NewItemCard()
   > SetVectorParameterValue(self=@n2, ParameterName=Rarity 1, Value=@n1)
   > SetVectorParameterValue(self=@n2, ParameterName=Rarity 2, Value=@n1.Color2)
  ]}
 HandleUpdateRarityVisibility {nodes [
   n1 = HandleUpdateRarityVisibility!() -> (then, bIsVisible:bool=~) {is=title K2Node_FunctionEntry}
   > Set bRarityVisible(bRarityVisible:bool=@n1) -> (Output_Get:bool)
   > SetScalarParameterValue(
    self=Get ImageMaterial_NewItemCard()
    ParameterName=Angle
    Value=SelectFloat(A=Get InitialAngle() -> (InitialAngle:double), B=0, bPickA=Get bRarityVisible())
   )
   > SetScalarParameterValue(
    self=Get ImageMaterial_NewItemCard()
    ParameterName=Angle Offset
    Value=SelectFloat(A=Get InitialAngleOffset() -> (InitialAngleOffset:double), B=1, bPickA=Get bRarityVisible())
   )
  ]}
 HandleUpdateVisuals {nodes [
   HandleUpdateVisuals!() -> (
    then
    bIsItem:bool=~
    bIsBorderAnimated:bool=~
    IsHero:bool=~
    IsReward:bool=~
    IsShuffleItem:bool=~
    ShowAsArchived:bool=~
    UseAnimations:bool=~
   ) {is=title K2Node_FunctionEntry}
   > SetScalarParameterValue(self=Get ImageMaterial_NewItemCard(), ParameterName=bIsReward, Value="To Float (Boolean)"(InBool=Get IsReward() -> (IsReward:bool)))
   > SetScalarParameterValue(
    self=Get ImageMaterial_NewItemCard()
    ParameterName=bIsShuffIeItem
    Value="To Float (Boolean)"(InBool=Get IsShuffleItem() -> (IsShuffleItem:bool))
   )
   > SetScalarParameterValue(self=Get ImageMaterial_NewItemCard(), ParameterName=RandomByWidget, Value=RandomFloatInRange(Min=0, Max=1))
   > SetScalarParameterValue(self=Get ImageMaterial_NewItemCard(), ParameterName=RandomByWidget2, Value=RandomFloatInRange(Min=0, Max=1))
   > ArchiveAnimation!(execute, self, IsArchived:bool=Get ShowAsArchived() -> (ShowAsArchived:bool), UseAnimations:bool=Get UseAnimations() -> (UseAnimations:bool)) -> (then) {is=title K2Node_CallFunction}
  ]}
 Sequencer Events {nodes [
   UpdateImageIconScale() {is=title K2Node_CustomEvent}
   n1 = Get AnimationIconScale() -> (AnimationIconScale:double)
   n2 = "float > float"(A=@n1)
   > SetScalarParameterValue(
    self=Get ImageMaterial_NewItemCard()
    ParameterName=ScalePreviewImage
    Value=float * float(A=Get BaseIconScale() -> (BaseIconScale:double), B=SelectFloat(A=@n1, B=1, bPickA=@n2)) {is=title K2Node_PromotableOperator}
   )
  ]}
 Update Icon {nodes [
   n1 = Update Icon!() -> (then, Icon:Texture 2D Object Reference, Scale:double=~) {is=title K2Node_FunctionEntry}
   > SetTextureParameterValue(self=Get ImageMaterial_NewItemCard(), ParameterName=ItemDisplay, Value=@n1)
   > n2 = Set BaseIconScale(BaseIconScale:double=@n1.Scale) -> (Output_Get:double)
   > SetScalarParameterValue(self=Get ImageMaterial_NewItemCard(), ParameterName=ScalePreviewImage, Value=@n2)
  ]}
]
variables [
 AnimationIconScale {type=number, default=1}
 BaseIconScale {type=number, default=1}
 bRarityVisible {type=boolean, default=true}
 CachedIcon {type=Texture2D}
 CachedScale {type=number}
 ImageMaterial_NewItemCard {type=MaterialInstanceDynamic}
 InitialAngle {type=number}
 InitialAngleOffset {type=number}
]
widgets {
 info {inheritedWidgetCount=3, parentClass=/Script/FortniteUI.FortCosmeticItemCard, rootWidgetClass=/Script/CommonUI.CommonLoadGuard, widgetCount=3}
 widgets [
  LoadGuard {
   widgetClassPath=&CommonLoadGuard
   bInherited=true
   bIsVariable=true
   widget=LoadGuard
   properties {loadingText="", textStyle=&TextStyle-Base_C}
  }
  SizeBox_NewItemCard {
   widgetClassPath=&SizeBox
   bInherited=true
   bIsVariable=true
   parent=LoadGuard
   widget=SizeBox_NewItemCard
   properties {bOverride_HeightOverride=true, bOverride_WidthOverride=true, heightOverride=200, widthOverride=128}
   slot {class=LoadGuardSlot}
  }
  Image_NewItemCard {
   widgetClassPath=&Image
   bInherited=true
   bIsVariable=true
   parent=SizeBox_NewItemCard
   widget=Image_NewItemCard
   properties {brush {
     bIsDynamicallyLoaded=false
     drawAs=Image
     imageType=NoImage
     mirroring=NoMirror
     resourceName=None
     resourceObject=/FortniteCoreUI/Athena/UI/Frontend/CosmeticItemCard/Materials/M_UI_ItemCard_V2_PlatformSeries
     tiling=NoTile
     imageSize=<32 32>
     margin=<ltrb 0 0 0 0>
     outlineSettings {
      bUseBrushTransparency=false
      roundingType=HalfHeightRadius
      width=0
      color {colorUseRule=UseColor_Specified, specifiedColor=<rgba 0 0 0 0>}
      cornerRadii=<0 0 0 0>
     }
     tintColor {colorUseRule=UseColor_Specified, specifiedColor=<rgba 1 1 1 1>}
     uVRegion {bIsValid=false, max=<0 0>, min=<0 0>}
    }}
   slot {class=SizeBoxSlot}
  }
 ]
}
